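# Material function that adds Fortnite gameplay visuals to input material attributes: destructible damage, fire system effects and Creative placement mode.
# By default it chains MF_FN_DestructibleDamage, MF_CurieFallbackEffect and MF_CreativeMoveTool.
# Static switch parameters toggle damage, its textures, hit wobble, and whether wobble adds to or overrides existing world position offset.
# The IsForLandscapeMaterial static bool instead selects only MF_Curie_Landscape, for Landscape and grass, skipping damage and move-tool visuals.
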MaterialFunction /Game/Creative/Materials/MaterialFunctions/MF_CreativeFortniteGameplay
graph {
 properties {description="This material function contains Fortnite gameplay related settings such as destructible damage, the fire system, and the Creative placement mode visuals."}
 nodes [
  n1 = FunctionInput(Preview=SetMaterialAttributes!(MaterialAttributes)) {bUsePreviewValueAsDefault=true, description=Material Input, inputType=FunctionInput_MaterialAttributes, sortPriority=-1000}
  n2 = MaterialFunctionCall!(
   In=@n1
   UseHeightFalloff=StaticBool() {desc="UseHeightFallof=True only works with RVT"}
   OnFireEmissiveOverride
   HeightFalloffRadius
   HeightFalloffHardness
   VariationMasksMobile
   In_Grass
   Is_Stylised_WE
  ) {
   desc="Landscape and LandscapeGrassType usage only!\r\nUses 2 texture samples on mobile"
   materialFunction=/Game/Packages/Fortress_SharedMaterials/Base_Material_Functions/MF_Curie_Landscape
  }
  n3 = StaticSwitchParameter(True=StaticBool() {value=true}, False=StaticBool()) {
   defaultValue=true
   desc=Enable the Destructible Damage visual effects for objects that can take damage. This consists of damage textures and an animated world position offset wobble.
   group=z_GameplayParameters_DestructibleDamage
   parameterName=Use_DestructibleDamage
   sortPriority=-100
  }
  n4 = StaticSwitchParameter(True=StaticBool() {value=true}, False=StaticBool()) {
   defaultValue=true
   desc=Setting this to false will not use any texture samplers for the DestructibleDamage.
   group=z_GameplayParameters_DestructibleDamage
   parameterName=Use_DestrucibleDamage_Textures
   sortPriority=-90
  }
  n5 = StaticSwitchParameter(True=StaticBool() {value=true}, False=StaticBool()) {
   defaultValue=true
   desc=Setting this to false will not use any animated wobble when the object is hit.
   group=z_GameplayParameters_DestructibleDamage
   parameterName=Use_DestructibleDamage_Wobble
   sortPriority=-80
  }
  n6 = StaticSwitchParameter(True=StaticBool() {value=true}, False=StaticBool()) {
   desc=Adds the destructible damage wobble to the world position offset. Setting this to false will simply override all world position offset that has been set before this material function.
   group=z_GameplayParameters_DestructibleDamage
   parameterName=Add_DestructibleDamage_Wobble to world position offset
   sortPriority=5
  }
  n7 = MaterialFunctionCall!(
   In=MaterialFunctionCall(Material_In=@n1, UseDestructibleDamage=@n3, UseCrackedDamageTextures=@n4, UseDamageWobbleWPO=@n5, UseAdditiveWPO=@n6) {materialFunction=&MF_FN_DestructibleDamage}.Material_Out
   UseSphericalMappingOverride
  ) {materialFunction=/Game/Packages/Fortress_SharedMaterials/Base_Material_Functions/MF_CurieFallbackEffect}
  n8 = FunctionInput(Preview=StaticBool()) {
   bUsePreviewValueAsDefault=true
   description="Set this to True if the material is used on Landscape or LandscapeGrassTypes. Default=False"
   inputName=IsForLandscapeMaterial
   inputType=FunctionInput_StaticBool
   sortPriority=-100
  }
  FunctionOutput(None=StaticSwitch(
    True=@n2.Result
    False=MaterialFunctionCall(MA In=@n7.Result) {materialFunction=&MF_CreativeMoveTool}.MA Out
    Value=@n8
   )) {
   desc="This material function contains Fortnite gameplay related settings such as destructible damage, the fire system, and the Creative placement mode visuals."
   description="This material function contains Fortnite gameplay related settings such as destructible damage, the fire system, and the Creative placement mode visuals."
  }
 ]
}
